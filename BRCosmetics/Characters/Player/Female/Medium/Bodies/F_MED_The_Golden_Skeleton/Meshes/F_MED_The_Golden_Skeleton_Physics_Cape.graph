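# Physics asset for the cape of the Golden Skeleton medium female player body, using capsule bodies on cape bones, torso, arms and legs.
# Cape chains of uneven length hang from cape_root and a left clavicle anchor on the upper spine, the longest on the left.
# Chain joints have limited swing and very tight twist, while body joints and anchors rotate freely.
# Free-rotation cross-links tie neighbouring chains together at their lower segments.
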PhysicsAsset /BRCosmetics/Characters/Player/Female/Medium/Bodies/F_MED_The_Golden_Skeleton/Meshes/F_MED_The_Golden_Skeleton_Physics_Cape
bodies {
 calf_l [{shapeType=Capsule, length=35.5, radius=8.5, center=<-24 -2 0>, rotation=<rot 90 0 0>}]
 calf_r [{shapeType=Capsule, length=35.5, radius=8.5, center=<24 2 0>, rotation=<rot 90 0 180>}]
 cape_root [{shapeType=Capsule, length=0.505, radius=0.505}]
 clavicle_cape_l [{shapeType=Capsule, length=0.505, radius=0.505}]
 dyn_capeMid_01 [{shapeType=Capsule, length=4, radius=5, shapeName=dyn_capeMid_01_capsule, center=<4 0 5>, rotation=<rot 0 0 90>}]
 dyn_capeMid_01_l [{shapeType=Capsule, length=4, radius=5, shapeName=dyn_capeMid_01_l_capsule, center=<4 0 5>, rotation=<rot 0 0 90>}]
 dyn_capeMid_01_r [{shapeType=Capsule, length=4, radius=5, shapeName=dyn_capeMid_01_r_capsule, center=<4 0 5>, rotation=<rot 0 0 90>}]
 dyn_capeMid_02 [{shapeType=Capsule, length=8, radius=6, shapeName=dyn_capeMid_02_capsule, center=<8 0 5>, rotation=<rot 90 0 0>}]
 dyn_capeMid_02_l [{shapeType=Capsule, length=8, radius=6, shapeName=dyn_capeMid_02_l_capsule, center=<8 0 5>, rotation=<rot 90 0 0>}]
 dyn_capeMid_02_r [{shapeType=Capsule, length=8, radius=6, shapeName=dyn_capeMid_02_r_capsule, center=<8 0 4>, rotation=<rot 90 0 0>}]
 dyn_capeMid_03 [{shapeType=Capsule, length=8, radius=6, shapeName=dyn_capeMid_03_capsule, center=<6 0 4>, rotation=<rot 90 0 0>}]
 dyn_capeMid_03_l [{shapeType=Capsule, length=8, radius=6, shapeName=dyn_capeMid_03_l_capsule, center=<6 0 4>, rotation=<rot 90 0 0>}]
 dyn_capeMid_03_r [{shapeType=Capsule, length=7.75, radius=6.25, shapeName=dyn_capeMid_03_r_capsule, center=<6 0 4>, rotation=<rot 0 0 90>}]
 dyn_capeMid_04_l [{shapeType=Capsule, length=8, radius=6, shapeName=dyn_capeMid_04_l_capsule, center=<6 0 4>, rotation=<rot 90 0 0>}]
 dyn_cape_01_l [{shapeType=Capsule, length=4, radius=5, shapeName=dyn_cape_01_l_capsule, center=<4 0 5>, rotation=<rot 0 0 85>}]
 dyn_cape_01_r [{shapeType=Capsule, length=4, radius=5, shapeName=dyn_cape_01_r_capsule, center=<4 0 5>, rotation=<rot 0 0 90>}]
 dyn_cape_02_l [{shapeType=Capsule, length=8, radius=6, shapeName=dyn_cape_02_l_capsule, center=<8 0 5>, rotation=<rot 90 2.5304 2.5302>}]
 dyn_cape_02_r [{shapeType=Capsule, length=8, radius=6, shapeName=dyn_cape_02_r_capsule, center=<8 0 4>, rotation=<rot 90 0 0>}]
 dyn_cape_03_l [{shapeType=Capsule, length=8, radius=6, shapeName=dyn_cape_03_l_capsule, center=<6 0 4>, rotation=<rot 90 179.9803 179.98>}]
 dyn_cape_04_l [{shapeType=Capsule, length=8, radius=6, shapeName=dyn_cape_04_l_capsule, center=<6 0 4>, rotation=<rot 90 -10.025 -10.0249>}]
 dyn_cape_05_l [{shapeType=Capsule, length=8, radius=6, shapeName=dyn_cape_05_l_capsule, center=<6 0 4>, rotation=<rot 90 10.0248 10.0239>}]
 dyn_cape_06_l [{shapeType=Capsule, length=8, radius=6, shapeName=dyn_cape_06_l_capsule, center=<6 0 4>, rotation=<rot 90 179.9803 179.98>}]
 dyn_clothBack_04 [{shapeType=Capsule, length=7.25, radius=8, center=<0 0 -5.4195>, rotation=<rot 90 0 0>}]
 lowerarm_l [{shapeType=Capsule, length=25, radius=7, center=<14 1 -1>, rotation=<rot 90 0 0>}]
 lowerarm_r [{shapeType=Capsule, length=25, radius=7, center=<-14 1 -1>, rotation=<rot 90 0 0>}]
 pelvis [{shapeType=Capsule, length=8.25, radius=16, center=<-0.9635 1.2256 0.5524>}]
 spine_01 [{shapeType=Capsule, length=10.5, radius=11.5, center=<4.4248 -1.0923 0>}]
 spine_04 [{shapeType=Capsule, length=9.75, radius=10.5, center=<1.0185 0.6057 0>}]
 spine_05 [{shapeType=Capsule, length=10.6946, radius=10.375, shapeName=spine_05_capsule, center=<5.2817 0 0>}]
 thigh_l [{shapeType=Capsule, length=25, radius=10, center=<-24 -0.5582 0.3973>, rotation=<rot 90 0 0>}]
 thigh_r [{shapeType=Capsule, length=25, radius=10, center=<24 0.5582 -0.3973>, rotation=<rot 90 0 180>}]
 upperarm_l [{shapeType=Capsule, length=25.5, radius=5.5, center=<14 0 0>, rotation=<rot 90 0 0>}]
 upperarm_r [{shapeType=Capsule, length=25.5, radius=5.5, center=<-14 0 0>, rotation=<rot 90 0 0>}]
}
constraints [
 {
  bone1Name=spine_01
  bone2Name=pelvis
  swing1LimitDegrees=45
  swing1Motion=Free
  swing2LimitDegrees=45
  swing2Motion=Free
  twistLimitDegrees=45
  twistMotion=Free
 }
 {
  bone1Name=spine_04
  bone2Name=spine_01
  swing1LimitDegrees=45
  swing1Motion=Free
  swing2LimitDegrees=45
  swing2Motion=Free
  twistLimitDegrees=45
  twistMotion=Free
 }
 {
  bone1Name=upperarm_l
  bone2Name=spine_05
  swing1LimitDegrees=45
  swing1Motion=Free
  swing2LimitDegrees=45
  swing2Motion=Free
  twistLimitDegrees=45
  twistMotion=Free
 }
 {
  bone1Name=lowerarm_l
  bone2Name=upperarm_l
  swing1LimitDegrees=45
  swing1Motion=Free
  swing2LimitDegrees=45
  swing2Motion=Free
  twistLimitDegrees=45
  twistMotion=Free
 }
 {
  bone1Name=upperarm_r
  bone2Name=spine_05
  swing1LimitDegrees=45
  swing1Motion=Free
  swing2LimitDegrees=45
  swing2Motion=Free
  twistLimitDegrees=45
  twistMotion=Free
 }
 {
  bone1Name=lowerarm_r
  bone2Name=upperarm_r
  swing1LimitDegrees=45
  swing1Motion=Free
  swing2LimitDegrees=45
  swing2Motion=Free
  twistLimitDegrees=45
  twistMotion=Free
 }
 {
  bone1Name=clavicle_cape_l
  bone2Name=spine_05
  swing1LimitDegrees=45
  swing1Motion=Free
  swing2LimitDegrees=45
  swing2Motion=Free
  twistLimitDegrees=45
  twistMotion=Free
 }
 {bone1Name=dyn_cape_01_l, bone2Name=clavicle_cape_l, swing1LimitDegrees=15, swing2LimitDegrees=15, twistLimitDegrees=2}
 {bone1Name=dyn_cape_02_l, bone2Name=dyn_cape_01_l, swing1LimitDegrees=20, swing2LimitDegrees=20, twistLimitDegrees=2}
 {bone1Name=dyn_cape_03_l, bone2Name=dyn_cape_02_l, swing1LimitDegrees=25, swing2LimitDegrees=45, twistLimitDegrees=2}
 {bone1Name=dyn_cape_04_l, bone2Name=dyn_cape_03_l, swing1LimitDegrees=25, swing2LimitDegrees=45, twistLimitDegrees=2}
 {bone1Name=dyn_cape_05_l, bone2Name=dyn_cape_04_l, swing1LimitDegrees=25, swing2LimitDegrees=45, twistLimitDegrees=2}
 {bone1Name=dyn_cape_06_l, bone2Name=dyn_cape_05_l, swing1LimitDegrees=25, swing2LimitDegrees=45, twistLimitDegrees=2}
 {
  bone1Name=cape_root
  bone2Name=spine_05
  swing1LimitDegrees=45
  swing1Motion=Free
  swing2LimitDegrees=45
  swing2Motion=Free
  twistLimitDegrees=45
  twistMotion=Free
 }
 {bone1Name=dyn_capeMid_01_l, bone2Name=cape_root, swing1LimitDegrees=15, swing2LimitDegrees=15, twistLimitDegrees=2}
 {bone1Name=dyn_capeMid_02_l, bone2Name=dyn_capeMid_01_l, swing1LimitDegrees=20, swing2LimitDegrees=20, twistLimitDegrees=2}
 {bone1Name=dyn_capeMid_03_l, bone2Name=dyn_capeMid_02_l, swing1LimitDegrees=25, swing2LimitDegrees=45, twistLimitDegrees=2}
 {bone1Name=dyn_capeMid_04_l, bone2Name=dyn_capeMid_03_l, swing1LimitDegrees=25, swing2LimitDegrees=45, twistLimitDegrees=2}
 {bone1Name=dyn_capeMid_01, bone2Name=cape_root, swing1LimitDegrees=15, swing2LimitDegrees=15, twistLimitDegrees=2}
 {bone1Name=dyn_capeMid_02, bone2Name=dyn_capeMid_01, swing1LimitDegrees=20, swing2LimitDegrees=20, twistLimitDegrees=2}
 {bone1Name=dyn_capeMid_03, bone2Name=dyn_capeMid_02, swing1LimitDegrees=25, swing2LimitDegrees=45, twistLimitDegrees=2}
 {bone1Name=dyn_capeMid_03_r, bone2Name=dyn_capeMid_02_r, swing1LimitDegrees=45, swing2LimitDegrees=45, twistLimitDegrees=2}
 {bone1Name=dyn_cape_01_r, bone2Name=cape_root, swing1LimitDegrees=15, swing2LimitDegrees=15, twistLimitDegrees=2}
 {bone1Name=dyn_cape_02_r, bone2Name=dyn_cape_01_r, swing1LimitDegrees=20, swing2LimitDegrees=20, twistLimitDegrees=2}
 {bone1Name=dyn_capeMid_02_r, bone2Name=dyn_capeMid_01_r, swing1LimitDegrees=20, swing2LimitDegrees=20, twistLimitDegrees=2}
 {bone1Name=dyn_capeMid_01_r, bone2Name=cape_root, swing1LimitDegrees=15, swing2LimitDegrees=15, twistLimitDegrees=2}
 {
  bone1Name=dyn_capeMid_03_l
  bone2Name=dyn_cape_03_l
  swing1LimitDegrees=45
  swing1Motion=Free
  swing2LimitDegrees=45
  swing2Motion=Free
  twistLimitDegrees=45
  twistMotion=Free
 }
 {
  bone1Name=dyn_capeMid_04_l
  bone2Name=dyn_cape_04_l
  swing1LimitDegrees=45
  swing1Motion=Free
  swing2LimitDegrees=45
  swing2Motion=Free
  twistLimitDegrees=45
  twistMotion=Free
 }
 {
  bone1Name=dyn_capeMid_03
  bone2Name=dyn_capeMid_03_l
  swing1LimitDegrees=45
  swing1Motion=Free
  swing2LimitDegrees=45
  swing2Motion=Free
  twistLimitDegrees=45
  twistMotion=Free
 }
 {
  bone1Name=dyn_capeMid_03_r
  bone2Name=dyn_capeMid_03
  swing1LimitDegrees=45
  swing1Motion=Free
  swing2LimitDegrees=45
  swing2Motion=Free
  twistLimitDegrees=45
  twistMotion=Free
 }
 {
  bone1Name=thigh_l
  bone2Name=pelvis
  swing1LimitDegrees=45
  swing1Motion=Free
  swing2LimitDegrees=45
  swing2Motion=Free
  twistLimitDegrees=45
  twistMotion=Free
 }
 {
  bone1Name=calf_l
  bone2Name=thigh_l
  swing1LimitDegrees=45
  swing1Motion=Free
  swing2LimitDegrees=45
  swing2Motion=Free
  twistLimitDegrees=45
  twistMotion=Free
 }
 {
  bone1Name=thigh_r
  bone2Name=pelvis
  swing1LimitDegrees=45
  swing1Motion=Free
  swing2LimitDegrees=45
  swing2Motion=Free
  twistLimitDegrees=45
  twistMotion=Free
 }
 {
  bone1Name=calf_r
  bone2Name=thigh_r
  swing1LimitDegrees=45
  swing1Motion=Free
  swing2LimitDegrees=45
  swing2Motion=Free
  twistLimitDegrees=45
  twistMotion=Free
 }
 {
  bone1Name=dyn_clothBack_04
  bone2Name=pelvis
  swing1LimitDegrees=45
  swing1Motion=Free
  swing2LimitDegrees=45
  swing2Motion=Free
  twistLimitDegrees=45
  twistMotion=Free
 }
 {
  bone1Name=spine_05
  bone2Name=spine_04
  swing1LimitDegrees=45
  swing1Motion=Free
  swing2LimitDegrees=45
  swing2Motion=Free
  twistLimitDegrees=45
  twistMotion=Free
 }
]
