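# Physics asset for the cape pack of the female medium Golden Skeleton player body, setting up cape bone physics.
# Capsules cover pelvis, spine, arms, legs and one back-cloth bone, linked by free-motion constraints.
# Asymmetric cape chains, longer on the left, hang from cape_root and clavicle_cape_l, both attached to spine_05.
# Chain joints limit twist tightly, with wide swing near the root and narrow single-axis swing further down.
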PhysicsAsset /BRCosmetics/Characters/Player/Female/Medium/Bodies/F_MED_The_Golden_Skeleton/Meshes/F_MED_The_Golden_Skeleton_Physics_CapePack
bodies {
 calf_l [{shapeType=Capsule, length=35, radius=9, center=<-24 -2 0>, rotation=<rot 90 0 0>}]
 calf_r [{shapeType=Capsule, length=35, radius=9, center=<24 2 0>, rotation=<rot 90 0 180>}]
 cape_root [{shapeType=Capsule, length=0.505, radius=0.505}]
 clavicle_cape_l [{shapeType=Capsule, length=0.505, radius=0.505}]
 dyn_capeMid_01 [{shapeType=Capsule, length=4, radius=3, center=<5 0 2>, rotation=<rot 90 0 0>}]
 dyn_capeMid_01_l [{shapeType=Capsule, length=4, radius=3, center=<5 0 2>, rotation=<rot 90 0 0>}]
 dyn_capeMid_01_r [{shapeType=Capsule, length=4, radius=3, center=<5 0 2>, rotation=<rot 90 0 0>}]
 dyn_capeMid_02 [{shapeType=Capsule, length=6, radius=4, center=<7 0 3>, rotation=<rot 90 0 0>}]
 dyn_capeMid_02_l [{shapeType=Capsule, length=6, radius=4, center=<7 0 3>, rotation=<rot 90 0 0>}]
 dyn_capeMid_02_r [{shapeType=Capsule, length=6, radius=4, center=<7 0 3>, rotation=<rot 90 0 0>}]
 dyn_capeMid_03 [{shapeType=Capsule, length=5.75, radius=6.25, center=<8 0 4>, rotation=<rot 90 0 0>}]
 dyn_capeMid_03_l [{shapeType=Capsule, length=6.125, radius=5.875, center=<8 0 4>, rotation=<rot 90 0 0>}]
 dyn_capeMid_03_r [{shapeType=Capsule, length=5.75, radius=6.25, center=<8 0 4>, rotation=<rot 90 0 0>}]
 dyn_capeMid_04_l [{shapeType=Capsule, length=6, radius=6, center=<8 0 4>, rotation=<rot 90 0 0>}]
 dyn_capeMid_05_l [{shapeType=Capsule, length=4, radius=6, center=<7 0 4>, rotation=<rot 90 0 0>}]
 dyn_cape_01_l [{shapeType=Capsule, length=4, radius=3, center=<5 0 2>, rotation=<rot 90 0 0>}]
 dyn_cape_01_r [{shapeType=Capsule, length=4, radius=3, center=<5 0 2>, rotation=<rot 90 0 0>}]
 dyn_cape_02_l [{shapeType=Capsule, length=6, radius=4, center=<7 0 3>, rotation=<rot 90 0 0>}]
 dyn_cape_02_r [{shapeType=Capsule, length=6, radius=4, center=<7 0 3>, rotation=<rot 90 0 0>}]
 dyn_cape_03_l [{shapeType=Capsule, length=5.75, radius=6.25, center=<8 0 4>, rotation=<rot 90 0 0>}]
 dyn_cape_04_l [{shapeType=Capsule, length=5.625, radius=6.375, center=<8 0 4>, rotation=<rot 90 0 0>}]
 dyn_cape_05_l [{shapeType=Capsule, length=3.625, radius=6.375, center=<7 0 4>, rotation=<rot 90 0 0>}]
 dyn_cape_06_l [{shapeType=Capsule, length=2.375, radius=6.375, center=<6 0 4>, rotation=<rot 90 0 0>}]
 dyn_clothBack_04 [{shapeType=Capsule, length=6.25, radius=9.75, center=<0 0 -5>, rotation=<rot 90 0 0>}]
 lowerarm_l [{shapeType=Capsule, length=25, radius=7, center=<14 1 -1>, rotation=<rot 90 0 0>}]
 lowerarm_r [{shapeType=Capsule, length=25, radius=7, center=<-14 1 -1>, rotation=<rot 90 0 0>}]
 pelvis [{shapeType=Capsule, length=8.5, radius=12.375, center=<0 -1.3973 0.1063>, rotation=<rot 0 -0.0001 -10>}]
 spine_01 [{shapeType=Capsule, length=15, radius=8}]
 spine_02 [{shapeType=Capsule, length=15, radius=8}]
 spine_03 [{shapeType=Capsule, length=15, radius=8}]
 spine_04 [{shapeType=Capsule, length=15, radius=8}]
 spine_05 [{shapeType=Capsule, length=15, radius=8.5}]
 thigh_l [{shapeType=Capsule, length=25, radius=10, center=<-24 -0.5582 0.3973>, rotation=<rot 90 0 0>}]
 thigh_r [{shapeType=Capsule, length=25, radius=10, center=<24 0.5582 -0.3973>, rotation=<rot 90 0 180>}]
 upperarm_l [{shapeType=Capsule, length=25, radius=6, center=<14 0 0>, rotation=<rot 90 0 0>}]
 upperarm_r [{shapeType=Capsule, length=25, radius=6, center=<-14 0 0>, rotation=<rot 90 0 0>}]
}
constraints [
 {
  bone1Name=spine_01
  bone2Name=pelvis
  swing1LimitDegrees=45
  swing1Motion=Free
  swing2LimitDegrees=45
  swing2Motion=Free
  twistLimitDegrees=45
  twistMotion=Free
 }
 {
  bone1Name=spine_02
  bone2Name=spine_01
  swing1LimitDegrees=45
  swing1Motion=Free
  swing2LimitDegrees=45
  swing2Motion=Free
  twistLimitDegrees=45
  twistMotion=Free
 }
 {
  bone1Name=spine_03
  bone2Name=spine_02
  swing1LimitDegrees=45
  swing1Motion=Free
  swing2LimitDegrees=45
  swing2Motion=Free
  twistLimitDegrees=45
  twistMotion=Free
 }
 {
  bone1Name=spine_04
  bone2Name=spine_03
  swing1LimitDegrees=45
  swing1Motion=Free
  swing2LimitDegrees=45
  swing2Motion=Free
  twistLimitDegrees=45
  twistMotion=Free
 }
 {
  bone1Name=spine_05
  bone2Name=spine_04
  swing1LimitDegrees=45
  swing1Motion=Free
  swing2LimitDegrees=45
  swing2Motion=Free
  twistLimitDegrees=45
  twistMotion=Free
 }
 {
  bone1Name=upperarm_l
  bone2Name=spine_05
  swing1LimitDegrees=45
  swing1Motion=Free
  swing2LimitDegrees=45
  swing2Motion=Free
  twistLimitDegrees=45
  twistMotion=Free
 }
 {
  bone1Name=lowerarm_l
  bone2Name=upperarm_l
  swing1LimitDegrees=45
  swing1Motion=Free
  swing2LimitDegrees=45
  swing2Motion=Free
  twistLimitDegrees=45
  twistMotion=Free
 }
 {
  bone1Name=upperarm_r
  bone2Name=spine_05
  swing1LimitDegrees=45
  swing1Motion=Free
  swing2LimitDegrees=45
  swing2Motion=Free
  twistLimitDegrees=45
  twistMotion=Free
 }
 {
  bone1Name=lowerarm_r
  bone2Name=upperarm_r
  swing1LimitDegrees=45
  swing1Motion=Free
  swing2LimitDegrees=45
  swing2Motion=Free
  twistLimitDegrees=45
  twistMotion=Free
 }
 {
  bone1Name=clavicle_cape_l
  bone2Name=spine_05
  swing1LimitDegrees=45
  swing1Motion=Free
  swing2LimitDegrees=45
  swing2Motion=Free
  twistLimitDegrees=45
  twistMotion=Free
 }
 {bone1Name=dyn_cape_01_l, bone2Name=clavicle_cape_l, swing1LimitDegrees=45, swing2LimitDegrees=45, twistLimitDegrees=5}
 {bone1Name=dyn_cape_02_l, bone2Name=dyn_cape_01_l, swing1LimitDegrees=45, swing2LimitDegrees=45, twistLimitDegrees=5}
 {bone1Name=dyn_cape_03_l, bone2Name=dyn_cape_02_l, swing1LimitDegrees=5, swing2LimitDegrees=35, twistLimitDegrees=5}
 {bone1Name=dyn_cape_04_l, bone2Name=dyn_cape_03_l, swing1LimitDegrees=5, swing2LimitDegrees=35, twistLimitDegrees=5}
 {bone1Name=dyn_cape_05_l, bone2Name=dyn_cape_04_l, swing1LimitDegrees=5, swing2LimitDegrees=35, twistLimitDegrees=5}
 {bone1Name=dyn_cape_06_l, bone2Name=dyn_cape_05_l, swing1LimitDegrees=5, swing2LimitDegrees=35, twistLimitDegrees=5}
 {
  bone1Name=cape_root
  bone2Name=spine_05
  swing1LimitDegrees=45
  swing1Motion=Free
  swing2LimitDegrees=45
  swing2Motion=Free
  twistLimitDegrees=45
  twistMotion=Free
 }
 {bone1Name=dyn_capeMid_01_l, bone2Name=cape_root, swing1LimitDegrees=45, swing2LimitDegrees=45, twistLimitDegrees=5}
 {bone1Name=dyn_capeMid_02_l, bone2Name=dyn_capeMid_01_l, swing1LimitDegrees=45, swing2LimitDegrees=45, twistLimitDegrees=5}
 {bone1Name=dyn_capeMid_03_l, bone2Name=dyn_capeMid_02_l, swing1LimitDegrees=5, swing2LimitDegrees=35, twistLimitDegrees=5}
 {bone1Name=dyn_capeMid_04_l, bone2Name=dyn_capeMid_03_l, swing1LimitDegrees=5, swing2LimitDegrees=35, twistLimitDegrees=5}
 {bone1Name=dyn_capeMid_05_l, bone2Name=dyn_capeMid_04_l, swing1LimitDegrees=5, swing2LimitDegrees=35, twistLimitDegrees=5}
 {bone1Name=dyn_capeMid_01, bone2Name=cape_root, swing1LimitDegrees=45, swing2LimitDegrees=45, twistLimitDegrees=5}
 {bone1Name=dyn_capeMid_02, bone2Name=dyn_capeMid_01, swing1LimitDegrees=45, swing2LimitDegrees=45, twistLimitDegrees=5}
 {bone1Name=dyn_capeMid_03, bone2Name=dyn_capeMid_02, swing1LimitDegrees=5, swing2LimitDegrees=35, twistLimitDegrees=5}
 {bone1Name=dyn_capeMid_03_r, bone2Name=dyn_capeMid_02_r, swing1LimitDegrees=5, swing2LimitDegrees=35, twistLimitDegrees=5}
 {bone1Name=dyn_cape_01_r, bone2Name=cape_root, swing1LimitDegrees=45, swing2LimitDegrees=45, twistLimitDegrees=5}
 {bone1Name=dyn_cape_02_r, bone2Name=dyn_cape_01_r, swing1LimitDegrees=45, swing2LimitDegrees=45, twistLimitDegrees=5}
 {bone1Name=dyn_capeMid_02_r, bone2Name=dyn_capeMid_01_r, swing1LimitDegrees=45, swing2LimitDegrees=45, twistLimitDegrees=5}
 {bone1Name=dyn_capeMid_01_r, bone2Name=cape_root, swing1LimitDegrees=45, swing2LimitDegrees=45, twistLimitDegrees=5}
 {
  bone1Name=dyn_clothBack_04
  bone2Name=pelvis
  swing1LimitDegrees=45
  swing1Motion=Free
  swing2LimitDegrees=45
  swing2Motion=Free
  twistLimitDegrees=45
  twistMotion=Free
 }
 {
  bone1Name=thigh_l
  bone2Name=pelvis
  swing1LimitDegrees=45
  swing1Motion=Free
  swing2LimitDegrees=45
  swing2Motion=Free
  twistLimitDegrees=45
  twistMotion=Free
 }
 {
  bone1Name=calf_l
  bone2Name=thigh_l
  swing1LimitDegrees=45
  swing1Motion=Free
  swing2LimitDegrees=45
  swing2Motion=Free
  twistLimitDegrees=45
  twistMotion=Free
 }
 {
  bone1Name=thigh_r
  bone2Name=pelvis
  swing1LimitDegrees=45
  swing1Motion=Free
  swing2LimitDegrees=45
  swing2Motion=Free
  twistLimitDegrees=45
  twistMotion=Free
 }
 {
  bone1Name=calf_r
  bone2Name=thigh_r
  swing1LimitDegrees=45
  swing1Motion=Free
  swing2LimitDegrees=45
  swing2Motion=Free
  twistLimitDegrees=45
  twistMotion=Free
 }
]
