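# Physics asset for the female medium Golden Skeleton player body, covering its back cloth and paired tassels.
# A four-capsule cloth chain hangs from a master box, and two three-capsule tassel chains share a base, with limited swing and tighter twist on each link.
# Pelvis, thigh and belt capsules supply body volumes.
# Both chain roots, the thighs and belt link to the pelvis, and a skirt-side body to the right thigh, with free rotation.
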PhysicsAsset /BRCosmetics/Characters/Player/Female/Medium/Bodies/F_MED_The_Golden_Skeleton/Meshes/F_MED_The_Golden_Skeleton_Physics_ClothBack
bodies {
 dyn_beltBase [{
   shapeType=Capsule
   length=12.5649
   radius=9.9322
   shapeName=dyn_beltBase_capsule
   center=<-14.8453 -0.2911 -3.6243>
   rotation=<rot 0 90.8479 -76.4876>
  }]
 dyn_clothBack_01 [{shapeType=Capsule, length=5, radius=6, shapeName=dyn_clothBack_01_capsule, center=<4 0 5>, rotation=<rot 0 0 90>}]
 dyn_clothBack_02 [{shapeType=Capsule, length=5, radius=6, shapeName=dyn_clothBack_02_capsule, center=<4 0 5>, rotation=<rot 0 0 90>}]
 dyn_clothBack_03 [{shapeType=Capsule, length=5, radius=6, shapeName=dyn_clothBack_03_capsule, center=<4 0 5>, rotation=<rot 0 0 90>}]
 dyn_clothBack_04 [{shapeType=Capsule, length=5, radius=6, shapeName=dyn_clothBack_04_capsule, center=<4 0 5>, rotation=<rot 0 0 90>}]
 dyn_clothBack_mstr [{shapeType=Box, extentX=1.01, extentY=1.01, extentZ=1.01}]
 dyn_skirtSide [{shapeType=Capsule, length=0.505, radius=0.505, shapeName=dyn_skirtSide_capsule, rotation=<rot 0 -90 0>}]
 dyn_tassel_01_l [{shapeType=Capsule, length=2, radius=2, shapeName=dyn_tassel_01_l_capsule, center=<2 0 0>, rotation=<rot 90 0 0>}]
 dyn_tassel_01_r [{shapeType=Capsule, length=2, radius=2, shapeName=dyn_tassel_01_r_capsule, center=<1 0 0>, rotation=<rot 90 0 0>}]
 dyn_tassel_02_l [{shapeType=Capsule, length=2, radius=2, shapeName=dyn_tassel_02_l_capsule, center=<2 0 0>, rotation=<rot 90 0 0>}]
 dyn_tassel_02_r [{shapeType=Capsule, length=2, radius=2, shapeName=dyn_tassel_02_r_capsule, center=<1 0 0>, rotation=<rot 90 0 0>}]
 dyn_tassel_03_l [{shapeType=Capsule, length=2, radius=2, shapeName=dyn_tassel_03_l_capsule, center=<2 0 0>, rotation=<rot 90 0 0>}]
 dyn_tassel_03_r [{shapeType=Capsule, length=2, radius=2, shapeName=dyn_tassel_03_r_capsule, center=<1 0 0>, rotation=<rot 90 0 0>}]
 dyn_tassel_base [{shapeType=Capsule, length=0.505, radius=0.505, shapeName=dyn_tassel_base_capsule, rotation=<rot 0 -90 0>}]
 pelvis [{shapeType=Capsule, length=10.25, radius=14.75, center=<-1.9303 1.8634 0>}]
 thigh_l [{shapeType=Capsule, length=25.5, radius=9.5, center=<-28 -2 0>, rotation=<rot 90 0 0>}]
 thigh_r [{shapeType=Capsule, length=25.5, radius=9.5, center=<28 2 0>, rotation=<rot 90 0 180>}]
}
constraints [
 {
  bone1Name=dyn_clothBack_mstr
  bone2Name=pelvis
  swing1LimitDegrees=45
  swing1Motion=Free
  swing2LimitDegrees=45
  swing2Motion=Free
  twistLimitDegrees=45
  twistMotion=Free
 }
 {bone1Name=dyn_clothBack_01, bone2Name=dyn_clothBack_mstr, swing1LimitDegrees=45, swing2LimitDegrees=45, twistLimitDegrees=5}
 {bone1Name=dyn_clothBack_02, bone2Name=dyn_clothBack_01, swing1LimitDegrees=45, swing2LimitDegrees=45, twistLimitDegrees=15}
 {bone1Name=dyn_clothBack_03, bone2Name=dyn_clothBack_02, swing1LimitDegrees=45, swing2LimitDegrees=45, twistLimitDegrees=15}
 {bone1Name=dyn_clothBack_04, bone2Name=dyn_clothBack_03, swing1LimitDegrees=45, swing2LimitDegrees=45, twistLimitDegrees=15}
 {
  bone1Name=thigh_l
  bone2Name=pelvis
  swing1LimitDegrees=45
  swing1Motion=Free
  swing2LimitDegrees=45
  swing2Motion=Free
  twistLimitDegrees=45
  twistMotion=Free
 }
 {
  bone1Name=thigh_r
  bone2Name=pelvis
  swing1LimitDegrees=45
  swing1Motion=Free
  swing2LimitDegrees=45
  swing2Motion=Free
  twistLimitDegrees=45
  twistMotion=Free
 }
 {
  bone1Name=dyn_tassel_base
  bone2Name=pelvis
  swing1LimitDegrees=45
  swing1Motion=Free
  swing2LimitDegrees=45
  swing2Motion=Free
  twistLimitDegrees=45
  twistMotion=Free
 }
 {bone1Name=dyn_tassel_01_l, bone2Name=dyn_tassel_base, swing1LimitDegrees=15, swing2LimitDegrees=10, twistLimitDegrees=2}
 {bone1Name=dyn_tassel_02_l, bone2Name=dyn_tassel_01_l, swing1LimitDegrees=15, swing2LimitDegrees=10, twistLimitDegrees=2}
 {bone1Name=dyn_tassel_03_l, bone2Name=dyn_tassel_02_l, swing1LimitDegrees=15, swing2LimitDegrees=10, twistLimitDegrees=2}
 {bone1Name=dyn_tassel_01_r, bone2Name=dyn_tassel_base, swing1LimitDegrees=15, swing2LimitDegrees=10, twistLimitDegrees=2}
 {bone1Name=dyn_tassel_02_r, bone2Name=dyn_tassel_01_r, swing1LimitDegrees=15, swing2LimitDegrees=10, twistLimitDegrees=2}
 {bone1Name=dyn_tassel_03_r, bone2Name=dyn_tassel_02_r, swing1LimitDegrees=15, swing2LimitDegrees=10, twistLimitDegrees=2}
 {
  bone1Name=dyn_beltBase
  bone2Name=pelvis
  swing1LimitDegrees=45
  swing1Motion=Free
  swing2LimitDegrees=45
  swing2Motion=Free
  twistLimitDegrees=45
  twistMotion=Free
 }
 {
  bone1Name=dyn_skirtSide
  bone2Name=thigh_r
  swing1LimitDegrees=45
  swing1Motion=Free
  swing2LimitDegrees=45
  swing2Motion=Free
  twistLimitDegrees=45
  twistMotion=Free
 }
]
