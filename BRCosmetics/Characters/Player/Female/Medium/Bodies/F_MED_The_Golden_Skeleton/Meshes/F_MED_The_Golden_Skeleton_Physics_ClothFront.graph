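# Physics asset for the front cloth bone chain on the female medium Golden Skeleton player body.
# Four dyn_clothFront capsule bodies link in sequence from a small box master body, which attaches to the pelvis with free angular motion.
# Chain joints have swing limits with very tight twist limits, so the cloth bends but barely twists.
# Oversized pelvis, thigh and calf capsules, joined by angularly free constraints, apparently act as body colliders for the cloth.
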PhysicsAsset /BRCosmetics/Characters/Player/Female/Medium/Bodies/F_MED_The_Golden_Skeleton/Meshes/F_MED_The_Golden_Skeleton_Physics_ClothFront
bodies {
 calf_l [{shapeType=Capsule, length=35.25, radius=9, center=<-21 -5 -1.6>, rotation=<rot 90 1.7 -3.5>}]
 calf_r [{shapeType=Capsule, length=35.25, radius=9, center=<21 5 1.6>, rotation=<rot 90 -90 84.8>}]
 dyn_clothFront_01 [{shapeType=Capsule, length=6.25, radius=4.5, shapeName=dyn_clothFront_01_capsule, center=<5 0 2>, rotation=<rot 90 0 0>}]
 dyn_clothFront_02 [{shapeType=Capsule, length=3.5, radius=4.5, shapeName=dyn_clothFront_02_capsule, center=<5 0 2>, rotation=<rot 0 0 90>}]
 dyn_clothFront_03 [{shapeType=Capsule, length=5.75, radius=5, shapeName=dyn_clothFront_03_capsule, center=<5 0 2>, rotation=<rot 90 0 0>}]
 dyn_clothFront_04 [{shapeType=Capsule, length=9.625, radius=5, shapeName=dyn_clothFront_04_capsule, center=<7 0 2>, rotation=<rot 90 0 0>}]
 dyn_clothFront_mstr [{shapeType=Box, extentX=1.01, extentY=1.01, extentZ=1.01}]
 pelvis [
  {shapeType=Capsule, length=37.9, radius=25, center=<-23 -28 0>, rotation=<rot -90 0 25>}
  {shapeType=Capsule, length=13, radius=17, center=<-5.5422 -11.6305 0.0001>}
 ]
 thigh_l [{shapeType=Capsule, length=25, radius=10.75, center=<-23.3441 -3.3701 0.08>, rotation=<rot 90 180 -179.5508>}]
 thigh_r [{shapeType=Capsule, length=24, radius=11.75, center=<23.3437 3.3142 -0.7701>, rotation=<rot 90 180 0.4492>}]
}
constraints [
 {
  bone1Name=dyn_clothFront_mstr
  bone2Name=pelvis
  swing1LimitDegrees=45
  swing1Motion=Free
  swing2LimitDegrees=45
  swing2Motion=Free
  twistLimitDegrees=45
  twistMotion=Free
 }
 {bone1Name=dyn_clothFront_01, bone2Name=dyn_clothFront_mstr, swing1LimitDegrees=35, swing2LimitDegrees=35, twistLimitDegrees=2}
 {bone1Name=dyn_clothFront_02, bone2Name=dyn_clothFront_01, swing1LimitDegrees=45, swing2LimitDegrees=45, twistLimitDegrees=5}
 {bone1Name=dyn_clothFront_03, bone2Name=dyn_clothFront_02, swing1LimitDegrees=45, swing2LimitDegrees=45, twistLimitDegrees=5}
 {bone1Name=dyn_clothFront_04, bone2Name=dyn_clothFront_03, swing1LimitDegrees=15, swing2LimitDegrees=35, twistLimitDegrees=5}
 {
  bone1Name=calf_l
  bone2Name=thigh_l
  swing1LimitDegrees=45
  swing1Motion=Free
  swing2LimitDegrees=45
  swing2Motion=Free
  twistLimitDegrees=45
  twistMotion=Free
 }
 {
  bone1Name=calf_r
  bone2Name=thigh_r
  swing1LimitDegrees=45
  swing1Motion=Free
  swing2LimitDegrees=45
  swing2Motion=Free
  twistLimitDegrees=45
  twistMotion=Free
 }
 {
  bone1Name=thigh_r
  bone2Name=pelvis
  swing1LimitDegrees=45
  swing1Motion=Free
  swing2LimitDegrees=45
  swing2Motion=Free
  twistLimitDegrees=45
  twistMotion=Free
 }
 {
  bone1Name=thigh_l
  bone2Name=pelvis
  swing1LimitDegrees=45
  swing1Motion=Free
  swing2LimitDegrees=45
  swing2Motion=Free
  twistLimitDegrees=45
  twistMotion=Free
 }
]
