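# Cosmetic AnimBP for the Golden Skeleton female body, adding procedural and simulated cloth, cape, belt and tassel motion to the pose copied from the parent mesh.
# Thigh and knee rotations read each update drive corrective bone transforms around RigidBody cloth simulations and AnimDynamics.
# A backbling check selects between Look At-shaped CapePack physics and normal cape physics, whose alpha drops when crouching.
# Several crouch-adjustment chains never reach the output pose.
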AnimBlueprint /BRCosmetics/Characters/Player/Female/Medium/Bodies/F_MED_The_Golden_Skeleton/Meshes/F_MED_The_Golden_Skeleton_AnimBP
target_skeleton=&/F_MED_The_Golden_Skeleton_Skeleton
graphs [
 AnimGraph {nodes [
   Anim Dynamics - Bone: dyn_beltBase(Alpha=0, ExternalForce="0,0,0") {
    angularDampingOverride=0.9999
    bAngularSpring=true
    bGravityOverrideInSimSpace=true
    bOverrideAngularDamping=true
    bUseGravityOverride=true
    simulationSpace=RootRelative
    alphaBoolBlend {blendOption=Linear}
    alphaScaleBiasClamp {scale=0.6}
    boundBone {boneName=dyn_beltBase}
    componentAppliedLinearAccClamp=<3000 3000 3000>
    componentLinearAccScale {z=0.35}
    componentLinearVelScale {z=0.35}
    gravityOverride {z=-980}
    physicsBodyDefinitions [{
      sphereCollisionRadius=0
      boundBone {boneName=dyn_beltBase}
      boxExtents=<20 20 10>
      constraintSetup {angularLimitsMax=<0 10 0>, angularLimitsMin=<0 0 0>, angularTarget=<0 0 0>}
      localJointOffset=<-3 0 -5>
     }]
    relativeSpaceBone {boneName=root}
   }
   n1 = Get bIsCrouching()
   "Constraint - dyn_clothBack_mstr (PP)"!(ComponentPose, bAlphaBoolEnabled:bool=@n1, ConstraintWeights_0=0.8, ConstraintWeights_1:float=0.2) -> (Pose) {
    alphaInputType=Bool
    alphaBoolBlend {blendOption=Linear}
    boneToModify {boneName=dyn_clothBack_mstr}
    constraintSetup [{targetBone {boneName=pelvis}}, {targetBone {boneName=thigh_r}}]
    constraintWeights [1,1]
   }
   n2 = "Transform (Modify) Bone - Bone: dyn_beltBase_mstr"!(ComponentPose, bAlphaBoolEnabled:bool=@n1, Translation="0,0,0", Rotation="-1,0.000,0.000", Scale="1,1,1") -> (Pose) {
    alphaInputType=Bool
    rotationMode=BMM_Additive
    alphaBoolBlend {blendOption=Linear}
    boneToModify {boneName=dyn_beltBase_mstr}
    rotation {pitch=-1}
   }
   n3 = "Transform (Modify) Bone - Bone: dyn_tasselUpper_mstr_r"!(ComponentPose=@n2, bAlphaBoolEnabled:bool=@n1, Translation="0,0,0", Rotation="40.000,0.000,-30.000", Scale="1,1,1") -> (Pose) {
    alphaInputType=Bool
    rotationMode=BMM_Additive
    alphaBoolBlend {blendOption=Linear}
    boneToModify {boneName=dyn_tasselUpper_mstr_r}
    rotation {pitch=40, roll=-30}
   }
   n4 = "Transform (Modify) Bone - Bone: dyn_tassel_base"!(ComponentPose=@n3, bAlphaBoolEnabled:bool=@n1, Translation="0,0,0", Rotation="0,0,10", Scale="1,1,1") -> (Pose) {
    alphaInputType=Bool
    rotationMode=BMM_Additive
    alphaBoolBlend {blendOption=Linear}
    boneToModify {boneName=dyn_tassel_base}
    rotation {roll=10}
   }
   n5 = "Transform (Modify) Bone - Bone: dyn_skirtSide_mstr"!(ComponentPose=@n4, bAlphaBoolEnabled:bool=@n1, Translation="-2,0.000,0.000", Rotation="-15,-15,0", Scale="1,1,1") -> (Pose) {
    alphaInputType=Bool
    rotationMode=BMM_Additive
    translationMode=BMM_Additive
    alphaBoolBlend {blendOption=Linear}
    boneToModify {boneName=dyn_skirtSide_mstr}
    rotation {pitch=-15, yaw=-15}
    translation {x=-2}
   }
   n6 = "Transform (Modify) Bone - Bone: dyn_beltSideA_mstr_r"!(ComponentPose=@n5, bAlphaBoolEnabled:bool=@n1, Translation="0.000,0.000,2.5", Rotation="0,0,0", Scale="1,1,1") -> (Pose) {
    alphaInputType=Bool
    translationMode=BMM_Additive
    alphaBoolBlend {blendOption=Linear}
    boneToModify {boneName=dyn_beltSideA_mstr_r}
    translation {z=2.5}
   }
   n7 = "Transform (Modify) Bone - Bone: dyn_beltSideB_mstr_r"!(ComponentPose=@n6, bAlphaBoolEnabled:bool=@n1, Translation="0,0.000,2.5", Rotation="0,0,0", Scale="1,1,1") -> (Pose) {
    alphaInputType=Bool
    translationMode=BMM_Additive
    alphaBoolBlend {blendOption=Linear}
    boneToModify {boneName=dyn_beltSideB_mstr_r}
    translation {z=2.5}
   }
   n8 = "Transform (Modify) Bone - Bone: dyn_beltFront_mstr_r"!(ComponentPose=@n7, bAlphaBoolEnabled:bool=@n1, Translation="0,0,1", Rotation="0,0,0", Scale="1,1,1") -> (Pose) {
    alphaInputType=Bool
    translationMode=BMM_Additive
    alphaBoolBlend {blendOption=Linear}
    boneToModify {boneName=dyn_beltFront_mstr_r}
    translation {z=1}
   }
   "Transform (Modify) Bone - Bone: dyn_clothFront_mstr"!(ComponentPose=@n8, bAlphaBoolEnabled:bool=@n1, Translation="0,0,0", Rotation="-25,0,-35", Scale="1,1,1") -> (Pose) {
    alphaInputType=Bool
    rotationMode=BMM_Additive
    alphaBoolBlend {blendOption=Linear}
    boneToModify {boneName=dyn_clothFront_mstr}
    rotation {pitch=-25, roll=-35}
   }
   "Transform (Modify) Bone - Bone: dyn_clothFront_mstr"!(
    ComponentPose="Transform (Modify) Bone - Bone: dyn_tassel_base"!(
     ComponentPose="Transform (Modify) Bone - Bone: dyn_beltBase_mstr"!(ComponentPose, Alpha:float=1, Translation="0,0,0", Rotation="0,0,0", Scale="1,1,1") -> (Pose) {rotationMode=BMM_Additive, alphaBoolBlend {blendOption=Linear}, boneToModify {boneName=dyn_beltBase_mstr}}
     Alpha:float=1
     Translation="0,0,0"
     Rotation="0,0,0"
     Scale="1,1,1"
    ) -> (Pose) {rotationMode=BMM_Additive, alphaBoolBlend {blendOption=Linear}, boneToModify {boneName=dyn_tassel_base}}
    Alpha:float=1
    Translation="0,0,0"
    Rotation="0,0,0"
    Scale="1,1,1"
   ) -> (Pose) {rotationMode=BMM_Additive, alphaBoolBlend {blendOption=Linear}, boneToModify {boneName=dyn_clothFront_mstr}}
   "Transform (Modify) Bone - Bone: dyn_clothBack_mstr"(Alpha=1, Translation="0.000,-0.976,0.000", Rotation="-10.000,0.000,0.000", Scale="1,1,1") {
    rotationMode=BMM_Additive
    rotationSpace=BCS_BoneSpace
    translationMode=BMM_Additive
    translationSpace=BCS_BoneSpace
    alphaBoolBlend {blendOption=Linear}
    boneToModify {boneName=dyn_clothBack_mstr}
    rotation {pitch=-10}
    translation {y=-0.976}
   }
   n9 = Get thigh_rot_up_l() -> (thigh_rot_up_l:double)
   n10 = "Transform (Modify) Bone - Bone: dyn_clothFront_mstr"(Alpha=@n9, Translation="0,0,0", Rotation="0,0,0", Scale="1,0.5,1") {
    rotationMode=BMM_Additive
    rotationSpace=BCS_BoneSpace
    scaleMode=BMM_Additive
    scaleSpace=BCS_BoneSpace
    translationMode=BMM_Additive
    translationSpace=BCS_BoneSpace
    alphaBoolBlend {blendOption=Linear}
    boneToModify {boneName=dyn_clothFront_mstr}
    scale {y=0.5}
   }
   "Transform (Modify) Bone - Bone: dyn_clothFront_01"(ComponentPose=@n10, Alpha=@n9, Translation="0,0,0", Rotation="0,0,0", Scale="1,2,1") {
    rotationMode=BMM_Additive
    rotationSpace=BCS_BoneSpace
    scaleMode=BMM_Additive
    scaleSpace=BCS_BoneSpace
    translationMode=BMM_Additive
    translationSpace=BCS_BoneSpace
    alphaBoolBlend {blendOption=Linear}
    boneToModify {boneName=dyn_clothFront_01}
    scale {y=2}
   }
   n11 = Get bIsCrouchSprinting()
   n12 = "Transform (Modify) Bone - Bone: dyn_clothFront_mstr"!(ComponentPose, bAlphaBoolEnabled:bool=@n11, Translation="-4,0.000,5", Rotation="60,0.000,0", Scale="1,1,1") -> (Pose) {
    alphaInputType=Bool
    rotationMode=BMM_Additive
    rotationSpace=BCS_BoneSpace
    translationMode=BMM_Additive
    translationSpace=BCS_BoneSpace
    alphaBoolBlend {blendOption=Linear}
    boneToModify {boneName=dyn_clothFront_mstr}
    rotation {pitch=60}
    translation {x=-4, z=5}
   }
   n13 = "Transform (Modify) Bone - Bone: dyn_beltFront_r"!(ComponentPose=@n12, bAlphaBoolEnabled:bool=@n11, Translation="1.851,-0.000,4.260", Rotation="0,0,0", Scale="1,1,1") -> (Pose) {
    alphaInputType=Bool
    translationMode=BMM_Additive
    translationSpace=BCS_BoneSpace
    alphaBoolBlend {blendOption=Linear}
    boneToModify {boneName=dyn_beltFront_r}
    translation {x=1.851, z=4.26}
   }
   n14 = "Transform (Modify) Bone - Bone: dyn_beltFront"!(ComponentPose=@n13, bAlphaBoolEnabled:bool=@n11, Translation="2.916,-0.000,6.026", Rotation="0,0,0", Scale="1,1,1") -> (Pose) {
    alphaInputType=Bool
    translationMode=BMM_Additive
    translationSpace=BCS_BoneSpace
    alphaBoolBlend {blendOption=Linear}
    boneToModify {boneName=dyn_beltFront}
    translation {x=2.916, z=6.026}
   }
   "Transform (Modify) Bone - Bone: dyn_beltSideA_r"!(ComponentPose=@n14, bAlphaBoolEnabled:bool=@n11, Translation="2.916,0.142,2.602", Rotation="0.000,0.000,10.001", Scale="1,1,1") -> (Pose) {
    alphaInputType=Bool
    rotationMode=BMM_Additive
    rotationSpace=BCS_BoneSpace
    translationMode=BMM_Additive
    translationSpace=BCS_BoneSpace
    alphaBoolBlend {blendOption=Linear}
    boneToModify {boneName=dyn_beltSideA_r}
    rotation {roll=10.001}
    translation=<2.916 0.142 2.602>
   }
   n15 = Get GravityOverride()
   RigidBody!(ComponentPose, Alpha, OverrideWorldGravity:Vector=@n15, ExternalForce) -> (Pose) {
    bClampLinearTranslationLimitToRefPose=true
    bOverrideWorldGravity=true
    overridePhysicsAsset=&/F_MED_The_Golden_Skeleton_Physics_CapePack
    simulationSpace=BaseBoneSpace
    alphaBoolBlend {blendOption=Linear}
    baseBoneRef {boneName=root}
    componentAppliedLinearAccClamp=<3000 3000 3000>
    componentLinearAccScale=<0.7 0.5 0.5>
    componentLinearVelScale=<0.7 0.5 0.5>
    overrideWorldGravity {z=-980}
    simSpaceSettings {
     maxAngularAcceleration=1000
     maxAngularVelocity=3
     maxLinearAcceleration=700
     maxLinearVelocity=500
     velocityScaleZ=0
     worldAlpha=1
     externalLinearDragV=<1 1>
    }
   }
   n16 = Look At - Bone: dyn_capeMid_02(ComponentPose=Local To Component(LocalPose=Use cached pose 'preCapeSim'()), Alpha=1) {
    alphaBoolBlend {blendOption=Linear}
    boneToModify {boneName=dyn_capeMid_02}
    lookAtLocation=<0 -17 -5>
    lookAtTarget {boneReference {boneName=pelvis}}
    lookAt_Axis {axis=<1 0>}
   }
   n17 = Look At - Bone: dyn_capeMid_02_r(ComponentPose=@n16, Alpha=1) {
    alphaBoolBlend {blendOption=Linear}
    boneToModify {boneName=dyn_capeMid_02_r}
    lookAtLocation=<0 -17 3>
    lookAtTarget {boneReference {boneName=pelvis}}
    lookAt_Axis {axis=<1 0>}
   }
   n18 = Look At - Bone: dyn_capeMid_02_l(ComponentPose=@n17, Alpha=1) {
    alphaBoolBlend {blendOption=Linear}
    boneToModify {boneName=dyn_capeMid_02_l}
    lookAtLocation=<0 -17 -10>
    lookAtTarget {boneReference {boneName=pelvis}}
    lookAt_Axis {axis=<1 0>}
   }
   n19 = Look At - Bone: dyn_cape_02_l(ComponentPose=@n18, Alpha=1) {
    alphaBoolBlend {blendOption=Linear}
    boneToModify {boneName=dyn_cape_02_l}
    lookAtLocation=<0 -17 -18>
    lookAtTarget {boneReference {boneName=pelvis}}
    lookAt_Axis {axis=<1 0>}
   }
   n20 = Look At - Bone: dyn_cape_02_r(ComponentPose=@n19, Alpha=1) {
    alphaBoolBlend {blendOption=Linear}
    boneToModify {boneName=dyn_cape_02_r}
    lookAtLocation=<0 -17 18>
    lookAtTarget {boneReference {boneName=pelvis}}
    lookAt_Axis {axis=<1 0>}
   }
   n21 = RigidBody(ComponentPose=@n20, OverrideWorldGravity=@n15) {
    bClampLinearTranslationLimitToRefPose=true
    bOverrideWorldGravity=true
    overridePhysicsAsset=&F_MED_The_Golden_Skeleton_Physics_CapePack
    simulationSpace=BaseBoneSpace
    alphaBoolBlend {blendOption=Linear}
    baseBoneRef {boneName=spine_02}
    componentAppliedLinearAccClamp=<3000 3000 3000>
    componentLinearAccScale {x=0.45}
    componentLinearVelScale {x=0.45}
    overrideWorldGravity {z=-980}
    simSpaceSettings {maxAngularAcceleration=200, maxAngularVelocity=5, maxLinearAcceleration=1000, maxLinearVelocity=100, velocityScaleZ=0, worldAlpha=1}
   }
   n22 = Component To Local(ComponentPose=@n21)
   n23 = RigidBody(
    ComponentPose=Local To Component(LocalPose=Use cached pose 'preCapeSim'())
    Alpha=Get Cape Crouch Alpha() -> (Cape Crouch Alpha:double)
    OverrideWorldGravity=@n15
   ) {
    bClampLinearTranslationLimitToRefPose=true
    bOverrideWorldGravity=true
    overridePhysicsAsset=&/F_MED_The_Golden_Skeleton_Physics_Cape
    simulationSpace=BaseBoneSpace
    alphaBoolBlend {blendOption=Linear}
    baseBoneRef {boneName=spine_02}
    componentAppliedLinearAccClamp=<3000 3000 3000>
    componentLinearAccScale {x=0.45}
    componentLinearVelScale {x=0.45}
    overrideWorldGravity {z=-980}
    simSpaceSettings {maxAngularAcceleration=200, maxAngularVelocity=5, maxLinearAcceleration=1000, maxLinearVelocity=100, velocityScaleZ=0, worldAlpha=1}
   }
   n24 = Component To Local(ComponentPose=@n23)
   n25 = Get isWearingBackbling()
   n26 = RigidBody(
    ComponentPose=Local To Component(LocalPose=Blend Poses by bool(BlendPose_0=@n22, BlendPose_1=@n24, bActiveValue=@n25) {blendType=Linear})
    OverrideWorldGravity=Get GravityOverride()
   ) {
    bClampLinearTranslationLimitToRefPose=true
    bOverrideWorldGravity=true
    overridePhysicsAsset=&/F_MED_The_Golden_Skeleton_Physics_ClothFront
    simulationSpace=BaseBoneSpace
    alphaBoolBlend {blendOption=Linear}
    baseBoneRef {boneName=spine_01}
    componentAppliedLinearAccClamp=<3000 3000 3000>
    componentLinearAccScale {x=0.3}
    componentLinearVelScale {x=0.3}
    overrideWorldGravity {z=-980}
    simSpaceSettings {
     maxAngularAcceleration=200
     maxAngularVelocity=5
     maxLinearAcceleration=1000
     maxLinearVelocity=200
     velocityScaleZ=0
     worldAlpha=1
     externalLinearDragV=<1 1>
    }
   }
   n27 = Anim Dynamics - Bone: dyn_skirtInner_02(ComponentPose=@n26, Alpha=1, ExternalForce="0,0,0") {
    angularDampingOverride=0.9999
    bAngularSpring=true
    bGravityOverrideInSimSpace=true
    bOverrideAngularDamping=true
    bUseGravityOverride=true
    simulationSpace=RootRelative
    alphaBoolBlend {blendOption=Linear}
    alphaScaleBiasClamp {scale=0.6}
    boundBone {boneName=dyn_skirtInner_02}
    componentAppliedLinearAccClamp=<3000 3000 3000>
    componentLinearAccScale=<0.35 0.35 0.35>
    componentLinearVelScale=<0.35 0.35 0.35>
    gravityOverride {z=-980}
    physicsBodyDefinitions [{
      sphereCollisionRadius=0
      boundBone {boneName=dyn_skirtInner_02}
      boxExtents=<10 30 20>
      constraintSetup {angularLimitsMax=<0 20 20>, angularLimitsMin=<0 -10 -20>, angularTarget=<0 0 0>}
      localJointOffset=<5 0 0>
     }]
    relativeSpaceBone {boneName=root}
   }
   n28 = Anim Dynamics - Bone: dyn_loop_l(ComponentPose=@n27, Alpha=1, ExternalForce="0,0,0") {
    angularDampingOverride=0.99
    angularSpringConstant=500
    bAngularSpring=true
    bGravityOverrideInSimSpace=true
    bOverrideAngularDamping=true
    bUseGravityOverride=true
    simulationSpace=BoneRelative
    alphaBoolBlend {blendOption=Linear}
    alphaScaleBiasClamp {scale=0.6}
    boundBone {boneName=dyn_loop_l}
    componentAppliedLinearAccClamp=<3000 3000 3000>
    componentLinearAccScale=<0.5 0.5 0.5>
    componentLinearVelScale=<0.5 0.5 0.5>
    gravityOverride {z=-980}
    physicsBodyDefinitions [{
      sphereCollisionRadius=0
      boundBone {boneName=dyn_loop_l}
      boxExtents=<10 15 20>
      constraintSetup {angularLimitsMax=<0 15 35>, angularLimitsMin=<0 -65 -35>, angularTarget=<0 0 0>}
      localJointOffset=<8 0 0>
     }]
   }
   n29 = Anim Dynamics - Bone: dyn_tassel_l(ComponentPose=@n28, Alpha=1, ExternalForce="0,0,0") {
    angularDampingOverride=0.99
    angularSpringConstant=500
    bAngularSpring=true
    bGravityOverrideInSimSpace=true
    bOverrideAngularDamping=true
    bUseGravityOverride=true
    simulationSpace=RootRelative
    alphaBoolBlend {blendOption=Linear}
    alphaScaleBiasClamp {scale=0.6}
    boundBone {boneName=dyn_tassel_l}
    componentAppliedLinearAccClamp=<3000 3000 3000>
    componentLinearAccScale=<0.5 0.5 0.5>
    componentLinearVelScale=<0.5 0.5 0.5>
    gravityOverride {z=-980}
    physicsBodyDefinitions [{
      sphereCollisionRadius=0
      boundBone {boneName=dyn_tassel_l}
      boxExtents=<10 20 20>
      constraintSetup {angularLimitsMax=<0 15 35>, angularLimitsMin=<0 -65 -35>, angularTarget=<0 0 0>}
      localJointOffset=<10 0 0>
     }]
   }
   n30 = Anim Dynamics - Bone: dyn_loop_r(ComponentPose=@n29, Alpha=1, ExternalForce="0,0,0") {
    angularDampingOverride=0.99
    angularSpringConstant=500
    bAngularSpring=true
    bGravityOverrideInSimSpace=true
    bOverrideAngularDamping=true
    bUseGravityOverride=true
    simulationSpace=BoneRelative
    alphaBoolBlend {blendOption=Linear}
    alphaScaleBiasClamp {scale=0.6}
    boundBone {boneName=dyn_loop_r}
    componentAppliedLinearAccClamp=<3000 3000 3000>
    componentLinearAccScale=<0.5 0.5 0.5>
    componentLinearVelScale=<0.5 0.5 0.5>
    gravityOverride {z=-980}
    physicsBodyDefinitions [{
      sphereCollisionRadius=0
      boundBone {boneName=dyn_loop_r}
      boxExtents=<10 15 20>
      constraintSetup {angularLimitsMax=<0 15 35>, angularLimitsMin=<0 -65 -35>, angularTarget=<0 0 0>}
      localJointOffset=<8 0 -1>
     }]
   }
   n31 = Anim Dynamics - Bone: dyn_tassel_r(ComponentPose=@n30, Alpha=1, ExternalForce="0,0,0") {
    angularDampingOverride=0.99
    angularSpringConstant=500
    bAngularSpring=true
    bGravityOverrideInSimSpace=true
    bOverrideAngularDamping=true
    bUseGravityOverride=true
    simulationSpace=RootRelative
    alphaBoolBlend {blendOption=Linear}
    alphaScaleBiasClamp {scale=0.6}
    boundBone {boneName=dyn_tassel_r}
    componentAppliedLinearAccClamp=<3000 3000 3000>
    componentLinearAccScale=<0.5 0.5 0.5>
    componentLinearVelScale=<0.5 0.5 0.5>
    gravityOverride {z=-980}
    physicsBodyDefinitions [{
      sphereCollisionRadius=0
      boundBone {boneName=dyn_tassel_r}
      boxExtents=<10 15 15>
      constraintSetup {angularLimitsMax=<0 35 35>, angularLimitsMin=<0 -35 -35>, angularTarget=<0 0 0>}
      localJointOffset=<5 -3 -1>
     }]
   }
   n32 = Anim Dynamics - Bone: dyn_tasselUpper_mstr_r(ComponentPose=@n31, Alpha=1, ExternalForce="0,0,0") {
    angularDampingOverride=0.99
    angularSpringConstant=500
    bAngularSpring=true
    bGravityOverrideInSimSpace=true
    bOverrideAngularDamping=true
    bUseGravityOverride=true
    simulationSpace=RootRelative
    alphaBoolBlend {blendOption=Linear}
    alphaScaleBiasClamp {scale=0.6}
    boundBone {boneName=dyn_tasselUpper_mstr_r}
    componentAppliedLinearAccClamp=<3000 3000 3000>
    componentLinearAccScale=<0.5 0.5 0.5>
    componentLinearVelScale=<0.5 0.5 0.5>
    gravityOverride {z=-980}
    physicsBodyDefinitions [{
      sphereCollisionRadius=0
      boundBone {boneName=dyn_tasselUpper_mstr_r}
      boxExtents=<10 30 20>
      constraintSetup {angularLimitsMax=<0 10 45>, angularLimitsMin=<0 -10 -5>, angularTarget=<0 0 0>}
      localJointOffset=<8 -2 -4>
     }]
   }
   n33 = "Transform (Modify) Bone - Bone: dyn_tasselUpper_r"(
    ComponentPose=@n32
    Alpha=Get knee rot r() -> (knee rot r:double)
    Translation="-0.221,0.678,-0.000"
    Rotation="0.000,45,0.000"
    Scale="1,1,1"
   ) {
    rotationMode=BMM_Additive
    rotationSpace=BCS_BoneSpace
    scaleMode=BMM_Additive
    scaleSpace=BCS_BoneSpace
    translationMode=BMM_Additive
    translationSpace=BCS_BoneSpace
    alphaBoolBlend {blendOption=Linear}
    boneToModify {boneName=dyn_tasselUpper_r}
    rotation {yaw=45}
    translation=<-0.2211 0.6782 0>
   }
   Output Pose(Result=Component To Local(ComponentPose=@n33)) {layerGroup=None}
   n34 = "Constraint - dyn_clothBack_mstr (PP)"!(ComponentPose=Local To Component(LocalPose=Use cached pose 'preSimCache'()), Alpha:float=1, ConstraintWeights_0=0.7, ConstraintWeights_1=0.3) -> (Pose) {
    alphaBoolBlend {blendOption=Linear}
    boneToModify {boneName=dyn_clothBack_mstr}
    constraintSetup [{targetBone {boneName=pelvis}}, {targetBone {boneName=thigh_r}}]
    constraintWeights [1,1]
   }
   n35 = RigidBody(ComponentPose=@n34, OverrideWorldGravity=Get GravityOverride()) {
    bClampLinearTranslationLimitToRefPose=true
    bOverrideWorldGravity=true
    overridePhysicsAsset=&/F_MED_The_Golden_Skeleton_Physics_ClothBack
    simulationSpace=BaseBoneSpace
    alphaBoolBlend {blendOption=Linear}
    baseBoneRef {boneName=spine_01}
    componentAppliedLinearAccClamp=<3000 3000 3000>
    componentLinearAccScale {x=0.3}
    componentLinearVelScale {x=0.3}
    overrideWorldGravity {z=-980}
    simSpaceSettings {
     maxAngularAcceleration=200
     maxAngularVelocity=5
     maxLinearAcceleration=1000
     maxLinearVelocity=200
     velocityScaleZ=0
     worldAlpha=1
     externalLinearDragV=<1 1>
    }
   }
   Save cached pose 'preCapeSim'(Pose=Component To Local(ComponentPose=@n35)) {cacheName=preCapeSim}
   n36 = "Constraint - dyn_clothFront_TGT (PP)"!(
    ComponentPose=Local To Component(LocalPose=Copy Pose From Mesh() {bUseAttachedParent=true})
    bAlphaBoolEnabled:bool=True
    ConstraintWeights_0=0.5
    ConstraintWeights_1:float=0.5
   ) -> (Pose) {
    alphaInputType=Bool
    alphaBoolBlend {blendOption=Linear}
    boneToModify {boneName=dyn_clothFront_TGT}
    constraintSetup [{targetBone {boneName=thigh_l}}, {targetBone {boneName=thigh_r}}]
    constraintWeights [1,1]
   }
   n37 = Look At - Bone: dyn_clothFront_01(ComponentPose=@n36, Alpha=1) {
    alphaBoolBlend {blendOption=Linear}
    boneToModify {boneName=dyn_clothFront_01}
    lookAtLocation {x=30, z=5}
    lookAtTarget {boneReference {boneName=dyn_clothFront_TGT}}
    lookAt_Axis {axis=<1 0>}
   }
   n38 = Get thigh_rot_up_l()
   n39 = "Transform (Modify) Bone - Bone: dyn_clothFront_01"(ComponentPose=@n37, Alpha=@n38, Translation="0.000,4,2", Rotation="0,0,0", Scale="1,1,1") {
    rotationMode=BMM_Additive
    rotationSpace=BCS_BoneSpace
    translationMode=BMM_Additive
    translationSpace=BCS_BoneSpace
    alphaBoolBlend {blendOption=Linear}
    boneToModify {boneName=dyn_clothFront_01}
    translation {y=4, z=2}
   }
   n40 = "Transform (Modify) Bone - Bone: def_clothFront_l"(ComponentPose=@n39, Alpha=@n38, Translation="-2,-2,2", Rotation="0,0,0", Scale="1,1,1") {
    rotationMode=BMM_Additive
    rotationSpace=BCS_BoneSpace
    translationMode=BMM_Additive
    translationSpace=BCS_BoneSpace
    alphaBoolBlend {blendOption=Linear}
    boneToModify {boneName=def_clothFront_l}
    translation=<-2 -2 2>
   }
   n41 = Get thigh_rot_up_r() -> (thigh_rot_up_r:double)
   n42 = "Transform (Modify) Bone - Bone: dyn_clothFront_01"(ComponentPose=@n40, Alpha=@n41, Translation="0.000,-4,2", Rotation="0,0,0", Scale="1,1,1") {
    rotationMode=BMM_Additive
    rotationSpace=BCS_BoneSpace
    translationMode=BMM_Additive
    translationSpace=BCS_BoneSpace
    alphaBoolBlend {blendOption=Linear}
    boneToModify {boneName=dyn_clothFront_01}
    translation {y=-4, z=2}
   }
   n43 = "Transform (Modify) Bone - Bone: def_clothFront_r"(ComponentPose=@n42, Alpha=@n41, Translation="-2,2,2", Rotation="0,0,0", Scale="1,1,1") {
    rotationMode=BMM_Additive
    rotationSpace=BCS_BoneSpace
    translationMode=BMM_Additive
    translationSpace=BCS_BoneSpace
    alphaBoolBlend {blendOption=Linear}
    boneToModify {boneName=def_clothFront_r}
    translation=<-2 2 2>
   }
   n44 = "Transform (Modify) Bone - Bone: dyn_clothFront_01"(
    ComponentPose=@n43
    Alpha=Get thight rot up all() -> (thight rot up all:double)
    Translation="-2.5,-0.000,6"
    Rotation="90,0,0"
    Scale="1,1,1"
   ) {
    rotationMode=BMM_Additive
    rotationSpace=BCS_BoneSpace
    translationMode=BMM_Additive
    translationSpace=BCS_BoneSpace
    alphaBoolBlend {blendOption=Linear}
    boneToModify {boneName=dyn_clothFront_01}
    rotation {pitch=90}
    translation {x=-2.5, z=6}
   }
   n45 = "Transform (Modify) Bone - Bone: dyn_beltBase_mstr"(ComponentPose=@n44, Alpha=Get thigh_rot_up_r(), Translation="0,0,0", Rotation="-12.903,-1.791,-5.870", Scale="0.94,1,1") {
    rotationMode=BMM_Additive
    rotationSpace=BCS_BoneSpace
    scaleMode=BMM_Additive
    scaleSpace=BCS_BoneSpace
    alphaBoolBlend {blendOption=Linear}
    boneToModify {boneName=dyn_beltBase_mstr}
    rotation=<rot -12.903 -1.7911 -5.87>
    scale {x=0.94}
   }
   n46 = "Transform (Modify) Bone - Bone: dyn_beltBase_mstr"(ComponentPose=@n45, Alpha=Get thigh_rot_otr_r() -> (thigh_rot_otr_r:double), Translation="0,0,0", Rotation="-5,0,0", Scale="1,1,1") {
    rotationMode=BMM_Additive
    rotationSpace=BCS_BoneSpace
    alphaBoolBlend {blendOption=Linear}
    boneToModify {boneName=dyn_beltBase_mstr}
    rotation {pitch=-5}
   }
   n47 = "Transform (Modify) Bone - Bone: clavicle_cape_l"!(ComponentPose=@n46, bAlphaBoolEnabled:bool=True, Translation="0.000,-3,0.000", Rotation="0,0,0", Scale="1,1,1") -> (Pose) {
    alphaInputType=Bool
    translationMode=BMM_Additive
    alphaBoolBlend {blendOption=Linear}
    boneToModify {boneName=clavicle_cape_l}
    translation {y=-3}
   }
   Save cached pose 'preSimCache'(Pose=Component To Local(ComponentPose=@n47)) {cacheName=preSimCache}
  ]}
 EventGraph {nodes [
   Event BlueprintUpdateAnimation()
   > n1 = Sequence!(execute) -> (then_0, then_1, then_2:Exec)
   > n2 = Gravity Override State Function DEPRECATED(CharacterPartAnimInstance=Self-Reference(), GameTimeInSeconds=GetGameTimeInSeconds(), GravityOverrideParams=Get Gravity Override Params())
   > Set GravityOverride(GravityOverride=@n2)
   n3 = GetOwnerSkeletalMeshComponent()
   n4 = BreakRotator(InRot=BreakTransform(InTransform=GetDeltaTransformFromRefPose(self=@n3, BoneName=thigh_r, BaseName=pelvis)).Rotation)
   Set thigh_rot_up_r(execute=@n1.then_1, thigh_rot_up_r:double=MapRangeClamped(Value=@n4.Yaw, InRangeA=-15, InRangeB=-100)) -> (Output_Get:double)
   > Set thigh_rot_otr_r(thigh_rot_otr_r:double=MapRangeClamped(Value=@n4.Pitch, InRangeA=0, InRangeB=30)) -> (Output_Get:double)
   n5 = BreakTransform(InTransform=GetDeltaTransformFromRefPose(self=GetOwnerSkeletalMeshComponent(), BoneName=thigh_l, BaseName=pelvis))
   > Set thigh_rot_up_l(thigh_rot_up_l:double=MapRangeClamped(Value=BreakRotator(InRot=@n5.Rotation).Yaw, InRangeA=-10, InRangeB=-100)) -> (Output_Get:double)
   n6 = BreakTransform(InTransform=GetDeltaTransformFromRefPose(self=GetOwnerSkeletalMeshComponent(), BoneName=thigh_r, BaseName=calf_r))
   > Set knee rot r(knee rot r:double=MapRangeClamped(Value=BreakRotator(InRot=@n6.Rotation).Yaw, InRangeA=0, InRangeB=-100)) -> (Output_Get:double)
   > Set thight rot up all(thight rot up all:double=float * float(A=Get thigh_rot_up_r(), B=Get thigh_rot_up_l())) -> (Output_Get:double)
   n7 = Branch(execute=@n1.then_2, Condition=Get bIsCrouching())
   > Set Cape Crouch Alpha(Cape Crouch Alpha:double=0.65) -> (Output_Get:double)
   Set Cape Crouch Alpha(execute=@n7.else, Cape Crouch Alpha=1)
   Event OnCharacterCustomizationLoaded()
   > Set isWearingBackbling(isWearingBackbling=CheckForBackbling())
   > Delay(Duration=0.15)
   > ResetDynamics(InTeleportType=TeleportPhysics)
   > ResetDynamics(InTeleportType=ResetPhysics)
  ]}
]
variables [
 Cape Crouch Alpha {type=number}
 Gravity Override Params {
  type=GravityOverrideParamsStruct
  default {
   skydive_DivingDown {gravityOverride=<0 0 0>}
   skydive_FloatingDown {gravityOverride=<0 0 0>, windFrequency=<150 700 800>}
   skydive_GoingUp {gravityOverride=<0 0 -100>, windFrequency=<150 700 800>, windOffset=<0 -600 400>}
  }
 }
 GravityOverride {type=Vector}
 isWearingBackbling {type=boolean}
 knee rot r {type=number}
 Skel_Mesh {type=SkeletalMeshComponent}
 thigh_rot_otr_r {type=number}
 thigh_rot_up_l {type=number}
 thigh_rot_up_r {type=number}
 thight rot up all {type=number}
]
