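# Physics asset for the Golden Skeleton female medium player body, set up for dynamic hair on the upper body.
# Capsules cover the upper spine, neck, head, clavicles and right upper arm only, plus front, strand, side and back hair bones.
# Hair chains branch from base bones under the head, with limited swing and very tight twist.
# Body joints and hair base attachments leave angular motion free.
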PhysicsAsset /BRCosmetics/Characters/Player/Female/Medium/Bodies/F_MED_The_Golden_Skeleton/Meshes/Parts/F_MED_The_Golden_Skeleton_Physics
bodies {
 clavicle_l [{shapeType=Capsule, length=3.75, radius=6.75, center=<5.544 -1.2447 -1.9153>, rotation=<rot 90 0 0>}]
 clavicle_r [{shapeType=Capsule, length=5.375, radius=6.5, center=<-5.39 1.2447 1.4608>, rotation=<rot 80 -0.0004 179.9996>}]
 dyn_hair_back_01 [{shapeType=Capsule, length=3, radius=5, shapeName=dyn_hair_back_01_capsule, center=<5 0 4>, rotation=<rot 90 0 0>}]
 dyn_hair_back_02 [{shapeType=Capsule, length=2.75, radius=6.25, shapeName=dyn_hair_back_02_capsule, center=<5 0 4>, rotation=<rot 85 0 0>}]
 dyn_hair_back_03 [{shapeType=Capsule, length=4.5, radius=6, shapeName=dyn_hair_back_03_capsule, center=<5 0 4>, rotation=<rot 85 0 0>}]
 dyn_hair_back_04 [{shapeType=Capsule, length=3.25, radius=6, shapeName=dyn_hair_back_04_capsule, center=<5 0 4>, rotation=<rot 80 0 0>}]
 dyn_hair_back_base [{shapeType=Capsule, length=0.505, radius=0.505, shapeName=dyn_hair_back_base_capsule, rotation=<rot 0 -90 0>}]
 dyn_hair_base [{shapeType=Capsule, length=0.505, radius=0.505, shapeName=dyn_hair_base_capsule, rotation=<rot 0 -90 0>}]
 dyn_hair_front_01_l [{shapeType=Capsule, length=2.25, radius=2.5, shapeName=dyn_hair_front_01_l_capsule, center=<3 0 1>, rotation=<rot 90 0 0>}]
 dyn_hair_front_01_r [{shapeType=Capsule, length=2.75, radius=2.5, shapeName=dyn_hair_front_01_r_capsule, center=<3 1.3037 1.5354>, rotation=<rot 90 0 0>}]
 dyn_hair_front_02_l [{shapeType=Capsule, length=4.25, radius=2.75, shapeName=dyn_hair_front_02_l_capsule, center=<4 0 0.4>, rotation=<rot 56.2953 39.5785 -6.3778>}]
 dyn_hair_front_02_r [{shapeType=Capsule, length=4.5, radius=2.75, shapeName=dyn_hair_front_02_r_capsule, center=<3.0158 1 1.5>, rotation=<rot 90 0 0>}]
 dyn_hair_front_03_r [{shapeType=Capsule, length=5.375, radius=2.875, shapeName=dyn_hair_front_03_r_capsule, center=<3 1 2>, rotation=<rot 74.9902 180 -175>}]
 dyn_hair_side_01_l [{shapeType=Capsule, length=3.25, radius=4.25, shapeName=dyn_hair_side_01_l_capsule, center=<6.1746 0 1.7929>, rotation=<rot 79.9996 0 0>}]
 dyn_hair_side_01_r [{shapeType=Capsule, length=2.75, radius=4.5, shapeName=dyn_hair_side_01_r_capsule, center=<5 0 3.5>, rotation=<rot 90 0 0>}]
 dyn_hair_side_02_l [{shapeType=Capsule, length=3.375, radius=5.875, shapeName=dyn_hair_side_02_l_capsule, center=<4.8825 2 1>, rotation=<rot -0.0001 20 89.9999>}]
 dyn_hair_side_02_r [{
   shapeType=Capsule
   length=3.875
   radius=5.25
   shapeName=dyn_hair_side_02_r_capsule
   center=<4 -3.1874 3.0627>
   rotation=<rot -0.0007 -0.0002 80.0002>
  }]
 dyn_hair_side_03_l [{shapeType=Capsule, length=4.25, radius=5, shapeName=dyn_hair_side_03_l_capsule, center=<5 0 2>, rotation=<rot 80.0001 -179.9999 -180>}]
 dyn_hair_side_03_r [{shapeType=Capsule, length=5.625, radius=4.625, shapeName=dyn_hair_side_03_r_capsule, center=<5 0 4>, rotation=<rot 80 0 0>}]
 dyn_hair_side_04_l [{shapeType=Capsule, length=3.5, radius=5.75, shapeName=dyn_hair_side_04_l_capsule, center=<6.6036 0 4.1403>, rotation=<rot 85.0001 -180 180>}]
 dyn_hair_side_l_base [{shapeType=Capsule, length=0.505, radius=0.505, shapeName=dyn_hair_side_l_base_capsule, rotation=<rot 0 -90 0>}]
 dyn_hair_side_r_base [{shapeType=Capsule, length=0.505, radius=0.505, shapeName=dyn_hair_side_r_base_capsule, rotation=<rot 0 -90 0>}]
 dyn_hair_strand_front_01 [{shapeType=Capsule, length=1.75, radius=1.5, center=<1.5 0 0.5>, rotation=<rot 90 0.0264 0.0264>}]
 head [{shapeType=Capsule, length=4, radius=5.5, center=<2.4 2.1 0>, rotation=<rot 90 0.0148 0.0149>}]
 neck_01 [{shapeType=Capsule, length=3.75, radius=5.45, center=<3 0 0>, rotation=<rot 90 0 0>}]
 spine_03 [{shapeType=Capsule, length=20.25, radius=12.75, center=<4.4081 -0.938 0>, rotation=<rot -0.0006 0 0>}]
 spine_04 [{shapeType=Capsule, length=21.75, radius=11.25, center=<2.8061 -0.816 0>, rotation=<rot -0.0006 0 0>}]
 spine_05 [{shapeType=Capsule, length=14.5, radius=8.375, center=<2.7027 -1.4413 0.0001>, rotation=<rot -0.0006 0 0>}]
 upperarm_r [{shapeType=Capsule, length=25.375, radius=3.625, center=<-14 0.9428 -1.226>, rotation=<rot 90 0 0>}]
}
constraints [
 {
  bone1Name=head
  bone2Name=neck_01
  swing1LimitDegrees=45
  swing1Motion=Free
  swing2LimitDegrees=45
  swing2Motion=Free
  twistLimitDegrees=45
  twistMotion=Free
 }
 {bone1Name=dyn_hair_front_01_r, bone2Name=dyn_hair_base, swing1LimitDegrees=30, swing2LimitDegrees=30, twistLimitDegrees=2}
 {bone1Name=dyn_hair_front_02_r, bone2Name=dyn_hair_front_01_r, swing1LimitDegrees=30, swing2LimitDegrees=30, twistLimitDegrees=2}
 {bone1Name=dyn_hair_front_03_r, bone2Name=dyn_hair_front_02_r, swing1LimitDegrees=30, swing2LimitDegrees=30, twistLimitDegrees=2}
 {bone1Name=dyn_hair_front_01_l, bone2Name=dyn_hair_base, swing1LimitDegrees=30, swing2LimitDegrees=30, twistLimitDegrees=2}
 {bone1Name=dyn_hair_front_02_l, bone2Name=dyn_hair_front_01_l, swing1LimitDegrees=30, swing2LimitDegrees=30, twistLimitDegrees=2}
 {bone1Name=dyn_hair_strand_front_01, bone2Name=dyn_hair_base, swing1LimitDegrees=25, swing2LimitDegrees=5, twistLimitDegrees=2}
 {
  bone1Name=upperarm_r
  bone2Name=clavicle_r
  swing1LimitDegrees=45
  swing1Motion=Free
  swing2LimitDegrees=45
  swing2Motion=Free
  twistLimitDegrees=45
  twistMotion=Free
 }
 {bone1Name=dyn_hair_side_01_l, bone2Name=dyn_hair_side_l_base, swing1LimitDegrees=5, swing2LimitDegrees=45, twistLimitDegrees=2}
 {bone1Name=dyn_hair_side_02_l, bone2Name=dyn_hair_side_01_l, swing1LimitDegrees=10, swing2LimitDegrees=65, twistLimitDegrees=2}
 {bone1Name=dyn_hair_side_03_l, bone2Name=dyn_hair_side_02_l, swing1LimitDegrees=10, swing2LimitDegrees=65, twistLimitDegrees=2}
 {bone1Name=dyn_hair_side_04_l, bone2Name=dyn_hair_side_03_l, swing1LimitDegrees=10, swing2LimitDegrees=65, twistLimitDegrees=2}
 {bone1Name=dyn_hair_side_01_r, bone2Name=dyn_hair_side_r_base, swing1LimitDegrees=5, swing2LimitDegrees=45, twistLimitDegrees=2}
 {bone1Name=dyn_hair_side_02_r, bone2Name=dyn_hair_side_01_r, swing1LimitDegrees=10, swing2LimitDegrees=65, twistLimitDegrees=2}
 {bone1Name=dyn_hair_side_03_r, bone2Name=dyn_hair_side_02_r, swing1LimitDegrees=10, swing2LimitDegrees=65, twistLimitDegrees=2}
 {bone1Name=dyn_hair_back_01, bone2Name=dyn_hair_back_base, swing1LimitDegrees=5, swing2LimitDegrees=45, twistLimitDegrees=2}
 {bone1Name=dyn_hair_back_02, bone2Name=dyn_hair_back_01, swing1LimitDegrees=10, swing2LimitDegrees=65, twistLimitDegrees=2}
 {bone1Name=dyn_hair_back_03, bone2Name=dyn_hair_back_02, swing1LimitDegrees=10, swing2LimitDegrees=65, twistLimitDegrees=2}
 {bone1Name=dyn_hair_back_04, bone2Name=dyn_hair_back_03, swing1LimitDegrees=10, swing2LimitDegrees=65, twistLimitDegrees=2}
 {
  bone1Name=neck_01
  bone2Name=spine_05
  swing1LimitDegrees=45
  swing1Motion=Free
  swing2LimitDegrees=45
  swing2Motion=Free
  twistLimitDegrees=45
  twistMotion=Free
 }
 {
  bone1Name=spine_05
  bone2Name=spine_04
  swing1LimitDegrees=45
  swing1Motion=Free
  swing2LimitDegrees=45
  swing2Motion=Free
  twistLimitDegrees=45
  twistMotion=Free
 }
 {
  bone1Name=clavicle_l
  bone2Name=spine_05
  swing1LimitDegrees=45
  swing1Motion=Free
  swing2LimitDegrees=45
  swing2Motion=Free
  twistLimitDegrees=45
  twistMotion=Free
 }
 {
  bone1Name=dyn_hair_base
  bone2Name=head
  swing1LimitDegrees=45
  swing1Motion=Free
  swing2LimitDegrees=45
  swing2Motion=Free
  twistLimitDegrees=45
  twistMotion=Free
 }
 {
  bone1Name=dyn_hair_back_base
  bone2Name=dyn_hair_base
  swing1LimitDegrees=45
  swing1Motion=Free
  swing2LimitDegrees=45
  swing2Motion=Free
  twistLimitDegrees=45
  twistMotion=Free
 }
 {
  bone1Name=dyn_hair_side_r_base
  bone2Name=dyn_hair_base
  swing1LimitDegrees=45
  swing1Motion=Free
  swing2LimitDegrees=45
  swing2Motion=Free
  twistLimitDegrees=45
  twistMotion=Free
 }
 {
  bone1Name=clavicle_r
  bone2Name=spine_05
  swing1LimitDegrees=45
  swing1Motion=Free
  swing2LimitDegrees=45
  swing2Motion=Free
  twistLimitDegrees=45
  twistMotion=Free
 }
 {
  bone1Name=dyn_hair_side_l_base
  bone2Name=dyn_hair_base
  swing1LimitDegrees=45
  swing1Motion=Free
  swing2LimitDegrees=45
  swing2Motion=Free
  twistLimitDegrees=45
  twistMotion=Free
 }
 {
  bone1Name=spine_04
  bone2Name=spine_03
  swing1LimitDegrees=45
  swing1Motion=Free
  swing2LimitDegrees=45
  swing2Motion=Free
  twistLimitDegrees=45
  twistMotion=Free
 }
]
