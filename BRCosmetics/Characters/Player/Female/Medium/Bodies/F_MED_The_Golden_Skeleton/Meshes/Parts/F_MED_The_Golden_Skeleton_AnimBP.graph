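# Post-copy AnimBlueprint for the Golden Skeleton female cosmetic part, adding hair motion to the parent mesh's pose.
# Additive bone offsets, weighted by head and right-arm rotations relative to spine_05, correct hair around look-at-aimed base bones and RigidBody simulation.
# Simulation runs in spine_05 space with gravity from a deprecated state function using skydive-specific wind settings, resetting when skydiving ends.
# Jump_Impulse is computed but not wired into the simulation.
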AnimBlueprint /BRCosmetics/Characters/Player/Female/Medium/Bodies/F_MED_The_Golden_Skeleton/Meshes/Parts/F_MED_The_Golden_Skeleton_AnimBP
target_skeleton=&/F_MED_The_Golden_Skeleton_Skeleton
graphs [
 AnimGraph {nodes [
   n1 = "Transform (Modify) Bone - Bone: dyn_hair_front_01_r"(
    ComponentPose=Local To Component(LocalPose=Copy Pose From Mesh() {bUseAttachedParent=true})
    Alpha=Get head_rot_bk() -> (head_rot_bk:double)
    Translation="0,0,0"
    Rotation="0,15,0"
    Scale="1,1,1"
   ) {
    rotationMode=BMM_Additive
    rotationSpace=BCS_BoneSpace
    translationMode=BMM_Additive
    translationSpace=BCS_BoneSpace
    alphaBoolBlend {blendOption=Linear}
    boneToModify {boneName=dyn_hair_front_01_r}
    rotation {yaw=15}
   }
   n2 = "Transform (Modify) Bone - Bone: dyn_hair_front_01_r"(ComponentPose=@n1, Alpha=Get arm_rot_up_r() -> (arm_rot_up_r:double), Translation="-1,0,-1", Rotation="-30,0,0", Scale="1,1,1") {
    rotationMode=BMM_Additive
    rotationSpace=BCS_BoneSpace
    translationMode=BMM_Additive
    translationSpace=BCS_BoneSpace
    alphaBoolBlend {blendOption=Linear}
    boneToModify {boneName=dyn_hair_front_01_r}
    rotation {pitch=-30}
    translation {x=-1, z=-1}
   }
   n3 = "Transform (Modify) Bone - Bone: dyn_hair_side_01_r"(ComponentPose=@n2, Alpha=Get head_rot_bk(), Translation="0,0,2", Rotation="0,0,0", Scale="1,1,1") {
    rotationMode=BMM_Additive
    rotationSpace=BCS_BoneSpace
    translationMode=BMM_Additive
    translationSpace=BCS_BoneSpace
    alphaBoolBlend {blendOption=Linear}
    boneToModify {boneName=dyn_hair_side_01_r}
    translation {z=2}
   }
   n4 = "Transform (Modify) Bone - Bone: dyn_hair_side_01_r"(ComponentPose=@n3, Alpha=Get head_rot_r() -> (head_rot_r:double), Translation="-3,2,3", Rotation="0,0,0", Scale="1,1,1") {
    rotationMode=BMM_Additive
    rotationSpace=BCS_BoneSpace
    translationMode=BMM_Additive
    translationSpace=BCS_BoneSpace
    alphaBoolBlend {blendOption=Linear}
    boneToModify {boneName=dyn_hair_side_01_r}
    translation=<-3 2 3>
   }
   n5 = Get head_rot_fd() -> (head_rot_fd:double)
   n6 = "Transform (Modify) Bone - Bone: dyn_hair_side_01_l"(ComponentPose=@n4, Alpha=@n5, Translation="0,-2,3", Rotation="0,0,0", Scale="1,1,1") {
    rotationMode=BMM_Additive
    rotationSpace=BCS_BoneSpace
    translationMode=BMM_Additive
    translationSpace=BCS_BoneSpace
    alphaBoolBlend {blendOption=Linear}
    boneToModify {boneName=dyn_hair_side_01_l}
    translation {y=-2, z=3}
   }
   n7 = "Transform (Modify) Bone - Bone: dyn_hair_back_01"(ComponentPose=@n6, Alpha=@n5, Translation="0,0,3", Rotation="0,0,0", Scale="1,1,1") {
    rotationMode=BMM_Additive
    rotationSpace=BCS_BoneSpace
    translationMode=BMM_Additive
    translationSpace=BCS_BoneSpace
    alphaBoolBlend {blendOption=Linear}
    boneToModify {boneName=dyn_hair_back_01}
    translation {z=3}
   }
   n8 = "Transform (Modify) Bone - Bone: dyn_hair_side_01_r"(ComponentPose=@n7, Alpha=@n5, Translation="0,2,3", Rotation="0,0,0", Scale="1,1,1") {
    rotationMode=BMM_Additive
    rotationSpace=BCS_BoneSpace
    translationMode=BMM_Additive
    translationSpace=BCS_BoneSpace
    alphaBoolBlend {blendOption=Linear}
    boneToModify {boneName=dyn_hair_side_01_r}
    translation {y=2, z=3}
   }
   n9 = Look At - Bone: dyn_hair_back_base(ComponentPose=@n8, Alpha=1) {
    alphaBoolBlend {blendOption=Linear}
    boneToModify {boneName=dyn_hair_back_base}
    lookAtLocation=<0 -14>
    lookAtTarget {boneReference {boneName=spine_05}}
    lookAt_Axis {axis=<1 0>}
   }
   n10 = Look At - Bone: dyn_hair_side_l_base(ComponentPose=@n9, Alpha=1) {
    alphaBoolBlend {blendOption=Linear}
    boneToModify {boneName=dyn_hair_side_l_base}
    lookAtLocation=<0 -18 -10>
    lookAtTarget {boneReference {boneName=spine_05}}
    lookAt_Axis {axis=<1 0>}
   }
   n11 = Look At - Bone: dyn_hair_side_r_base(ComponentPose=@n10, Alpha=1) {
    alphaBoolBlend {blendOption=Linear}
    boneToModify {boneName=dyn_hair_side_r_base}
    lookAtLocation=<0 -13 12>
    lookAtTarget {boneReference {boneName=spine_05}}
    lookAt_Axis {axis=<1 0>}
   }
   n12 = RigidBody!(ComponentPose=@n11, Alpha, OverrideWorldGravity:Vector=Get GravityOverride(), ExternalForce) -> (Pose) {
    bClampLinearTranslationLimitToRefPose=true
    bOverrideWorldGravity=true
    overridePhysicsAsset=&/F_MED_The_Golden_Skeleton_Physics
    simulationSpace=BaseBoneSpace
    alphaBoolBlend {blendOption=Linear}
    alphaScaleBiasClamp {bInterpResult=true, interpSpeedDecreasing=4, interpSpeedIncreasing=6}
    baseBoneRef {boneName=spine_05}
    componentAppliedLinearAccClamp=<3000 3000 3000>
    componentLinearAccScale=<0.3 0.3 0.3>
    componentLinearVelScale=<0.3 0.3 0.3>
    overrideWorldGravity {z=-875}
   }
   n13 = "Transform (Modify) Bone - Bone: dyn_hair_side_02_r"(ComponentPose=@n12, Alpha=Get arm_rot_up_r(), Translation="-2,1,1", Rotation="0,-10,0", Scale="1,1,1") {
    rotationMode=BMM_Additive
    rotationSpace=BCS_BoneSpace
    scaleMode=BMM_Additive
    scaleSpace=BCS_BoneSpace
    translationMode=BMM_Additive
    translationSpace=BCS_BoneSpace
    alphaBoolBlend {blendOption=Linear}
    boneToModify {boneName=dyn_hair_side_02_r}
    rotation {yaw=-10}
    translation=<-2 1 1>
   }
   Output Pose(Result=Component To Local(ComponentPose=@n13)) {layerGroup=None}
   Get Jump_Impulse() -> (Jump_Impulse:Vector)
  ]}
 EventGraph {nodes [
   Event BlueprintInitializeAnimation()
   > Set Skel_Mesh(Skel_Mesh=GetOwnerSkeletalMeshComponent())
   Event OnEndSkydiving()
   > ResetDynamics(InTeleportType=ResetPhysics)
   Event BlueprintUpdateAnimation()
   > n1 = Sequence!(execute) -> (then_0, then_1, then_2:Exec)
   n2 = NOT Boolean(A=OR Boolean!(A=Get bIsSkydiveDiveMode(), B=Get bIsSkyDiving(), C:bool=Get bIsSkydiveFloating()) -> (ReturnValue))
   n3 = OR Boolean!(A=Get bIsJumping(), B=Get bIsFalling()) -> (ReturnValue)
   > Set Jump_Impulse(Jump_Impulse:Vector=SelectVector(A="0, 0,1000", B="0, 0, 0", bPickA=AND Boolean(A=@n2, B=@n3))) -> (Output_Get:Vector)
   n4 = Gravity Override State Function DEPRECATED(
    execute=@n1.then_1
    CharacterPartAnimInstance=Self-Reference()
    GameTimeInSeconds=GetGameTimeInSeconds()
    GravityOverrideParams=Get Gravity Override Params()
   )
   > Set GravityOverride(GravityOverride=@n4)
   n5 = GetOwnerSkeletalMeshComponent()
   n6 = BreakRotator(InRot=BreakTransform(InTransform=GetDeltaTransformFromRefPose(self=@n5, BoneName=head, BaseName=spine_05)).Rotation)
   Set head_rot_bk(execute=@n1.then_2, head_rot_bk:double=MapRangeClamped(Value=@n6.Yaw, InRangeA=0, InRangeB=-50)) -> (Output_Get:double)
   > Set head_rot_r(head_rot_r:double=MapRangeClamped(Value=@n6.Pitch, InRangeA=0, InRangeB=50)) -> (Output_Get:double)
   > Set head_rot_fd(head_rot_fd:double=MapRangeClamped(Value=@n6.Yaw, InRangeA=0, InRangeB=15)) -> (Output_Get:double)
   n7 = BreakRotator(InRot=BreakTransform(InTransform=GetDeltaTransformFromRefPose(self=@n5, BoneName=upperarm_r, BaseName=spine_05)).Rotation)
   > Set arm_rot_up_r(arm_rot_up_r:double=MapRangeClamped(Value="Absolute (Float)"(A=@n7.Yaw), InRangeA=30, InRangeB=140)) -> (Output_Get:double)
  ]}
]
variables [
 arm_rot_up_r {type=number}
 Gravity Override Params {
  type=GravityOverrideParamsStruct
  default {
   skydive_DivingDown {gravityOverride=<0 0 0>}
   skydive_FloatingDown {gravityOverride=<0 0 -100>, windAmplitude=<200 -200 100>, windFrequency=<1.5 12 8>, windOffset=<0 -400 0>}
   skydive_GoingUp {gravityOverride=<0 0 -100>, windFrequency=<150 700 800>, windOffset=<0 -600 400>}
   skydive_Parachuting {windAmplitude=<200 -200 200>, windFrequency=<1.5 12 8>, windOffset=<0 -200 0>}
  }
 }
 GravityOverride {type=Vector}
 head_rot_bk {type=number}
 head_rot_fd {type=number}
 head_rot_r {type=number}
 Jump_Impulse {type=Vector}
 Skel_Mesh {type=SkeletalMeshComponent}
]
